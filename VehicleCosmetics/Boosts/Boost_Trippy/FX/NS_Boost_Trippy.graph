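# Niagara system for the Trippy vehicle boost cosmetic, trailing wavy multicolored ribbons and star sprites behind boost sockets on the car's skeletal mesh.
# Four ribbon emitters spawn per filtered socket, with GameModeID switching rate, lifetime and velocity.
# Ribbons write an age-growing sine offset to Particles.WPO and per-spawn-group R1/R2 widths; three cycle colors over life.
# One sprite emitter adds shrinking SubUV stars with curl noise and jitter.
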NiagaraSystem /VehicleCosmetics/Boosts/Boost_Trippy/FX/NS_Boost_Trippy
dependencies {
 usedDataInterfaces [
  &NiagaraDataInterfaceArrayFloat
  &NiagaraDataInterfaceSpriteRendererInfo
  &NiagaraDataInterfaceArrayInt32
  &NiagaraDataInterfaceCurve
  &NiagaraDataInterfaceColorCurve
  &NiagaraDataInterfaceSkeletalMesh
  &NiagaraDataInterfaceArrayFloat3
 ]
 usedDynamicInputs [&SelectVectorFromArray]
 usedModules [
  &self:SystemSpawnScript.SystemScriptSource.SystemScriptGraph.NiagaraNodeAssignment_0.SetVariables_#1
  &self:SystemSpawnScript.SystemScriptSource.SystemScriptGraph.NiagaraNodeAssignment_1.SetVariables_#2
  &SystemState
  &self:SystemSpawnScript.SystemScriptSource.SystemScriptGraph.NiagaraNodeAssignment_2.SetVariables_#3
 ]
}
emitters {
 BlueStars002 {
  emitterUpdateScript [
   &EmitterState()
   &VC_FilteredSocketCount()
   &SpawnRate(
    SpawnRate=&MakeCustomFloatFromBool(Bool=$User.bIsFrontendPreview, True Float=5, False Float=5)
    Use Spawn Probability=false
   )
  ]
  particleSpawnScript [
   &VC_PaintedVariantsOverride() {enabled=false}
   &InitializeParticle(
    Color=<rgba 2 0.71 1.3495 1>
    Lifetime Max=1
    Lifetime Min=0.8
    Lifetime Mode=::NewEnumerator1
    Sprite Rotation Mode=::NewEnumerator2
    Sprite Size Mode=::NewEnumerator1
    Uniform Sprite Size Max=29
    Uniform Sprite Size Min=10
    UsePositionOffset=false
    Write Material Random=true
   )
   &PartitionParticles(Number Of Partitions=$User.PortCount)
   &SkeletalMeshLocation(
    Bone / Socket Index=$Particles.PartitionParticles.Partition
    Bone / Socket Sampling Mode=::NewEnumerator5
    MaskSpawnGroup=false
    Mesh Sampling Type=::NewEnumerator6
    Sampled Position Offset=&AddVectorToPosition(Position=<-20 0 0>, Vector=$User.DTCarOffset)
    Skeletal Mesh=$System.SkeletalMesh
   )
   &ShapeLocation(
    Cylinder Height=66
    Cylinder Radius=30
    MaskSpawnGroup=false
    Override Seed=false
    Shape Primitive=::NewEnumerator1
    Surface Only=false
    UseOwnerScale=true
   )
   &SubUVAnimation(
    End Frame Range Override:NiagaraInt32=&SelectIntFromArray(Direct Array Index=&SpawnGroup(), Int Selection Array=DataInterface() {propertyValues {IntData [3,2,3,4]}})
    "Sprite Renderer "=DataInterface() {propertyValues {SpriteRenderer=&self:BlueStars002_1.NiagaraSpriteRendererProperties_3}}
    Start Frame Range Override=&SelectIntFromArray(Direct Array Index=$Emitter.SpawnGroup, Int Selection Array=DataInterface() {propertyValues {IntData [0,1,0,2]}})
    SubUV Animation Mode:ENiagara_SubUVLookupModeV2=::NewEnumerator2
    UseEndFrame=true
    UseStartFrame=true
   )
   &AddVelocity(Velocity=&RandomRangeVector(Minimum=<-250 0 5>, Maximum=<-200 1 -5>), Velocity Speed Scale=0.8)
   /VehicleCosmetics/FX/VC_FXParents/Modules/DLR_ProductPreviews(GameModeID:NiagaraInt32=$System.GameModeID, bPreviewVehicleSpeed:NiagaraBool=$System.bPreviewVehicleSpeed)
  ]
  particleUpdateScript [
   &ParticleState()
   &ScaleSpriteSize(Uniform Curve Sprite Scale=DataInterface() {propertyValues {Curve {
       defaultValue=0
       keys [{interpMode=RCIM_Cubic, tangentMode=RCTM_User, time=0.2141, value=1}, {interpMode=RCIM_Cubic, tangentMode=RCTM_User, time=1, value=0}]
      }}})
   &InheritSourceMovement(
    Applied Position Scale=&VectorFromFloat(Value=&FloatFromCurve(FloatCurve=DataInterface() {propertyValues {Curve {keys [{interpMode=RCIM_Cubic, tangentMode=RCTM_User, time=0, value=1}, {interpMode=RCIM_Cubic, tangentMode=RCTM_User, time=1, value=0}]}}}))
    ApplyForce=false
    ApplyPosition=true
    ApplyPositionFalloff=false
   )
   &Drag(
    Drag=&RandomRangeFloat(
     Minimum=0
     Maximum=&FloatFromCurve(FloatCurve=DataInterface() {propertyValues {Curve {
         defaultValue=0
         keys [{interpMode=RCIM_Cubic, tangentMode=RCTM_User, time=0, value=0}, {interpMode=RCIM_Cubic, tangentMode=RCTM_User, time=1, value=1}]
        }}})
    )
    Use Linear Drag=true
    Use Rotational Drag=true
   )
   &CurlNoiseForce(
    Bias Noise=false
    Noise Frequency=50000
    Noise Strength=&RandomRangeFloat(Minimum=100, Maximum=150)
    Pan Noise=false
   )
   &InheritSourceMovement(
    Applied Position Scale=&Multiply_VectorByFloat(
     Vector=<1 1 1>
     Float=&FloatFromCurve(FloatCurve=DataInterface() {propertyValues {Curve {
         defaultValue=1
         keys [
          {arriveTangent=-1.7885, interpMode=RCIM_Cubic, leaveTangent=-1.7885, tangentMode=RCTM_User, time=0.05, value=1}
          {interpMode=RCIM_Cubic, tangentMode=RCTM_User, time=1, value=0}
         ]
        }}})
    )
    ApplyForce=false
    ApplyPosition=true
    ApplyPositionFalloff=false
   ) {enabled=false, module=InheritSourceMovement001}
   /CascadeToNiagaraConverter/NiagaraScripts/ModuleScripts/RotatePosition(Rotation:Vector3f=<1.5885 0 0>)
   &JitterPosition(Jitter Amount:NiagaraFloat=&RandomRangeFloat(Minimum=-4, Maximum=4), Jitter Delay=2)
   &SolveForcesAndVelocity(Clamp Velocity=false, Limit Acceleration=false)
  ]
  rendererClasses [&NiagaraSpriteRendererProperties]
  renderers [{rendererClass=&NiagaraSpriteRendererProperties}]
 }
 RibbonTrail001 {
  emitterUpdateScript [
   &EmitterState()
   &VC_FilteredSocketCount()
   &VC_RateQualityHandler(Default Rate:NiagaraFloat=&MergeHelper_FloatSwitch(0=40, 1=40, 2=40, 3=60, Index=$System.GameModeID, Num=::NewEnumerator2))
   &VC_SpawnRibbon_v1(Number of Ribbons=$Emitter.PortCount, Spawn Time=-1, SpawnRate=$Emitter.Spawn Rate)
  ]
  particleSpawnScript [
   &VC_PaintedVariantsOverride() {enabled=false}
   &VC_InitializeRibbon_v1(
    Color=<rgba 0.55 0.55 0.55 0.8>
    Custom Offset=true
    Enable Skeletal Position=true
    Lifetime=&MergeHelper_FloatSwitch(0=0.74, 1=0.74, 2=0.74, 3=0.4, Index=$System.GameModeID, Num=::NewEnumerator2)
    Local Vector Offset=&AddVectorToPosition(Position=<-30 0 0>, Vector=$User.DTCarOffset)
    Offset Amount=1
    Ribbon Width=80
    Write Material Random=false
   )
   &AddVelocity(Velocity=&MergeHelper_VectorSwitch(0=<-1300 0 0>, 1=<-1300 0 0>, 2=<-1300 0 0>, 3=<-2300 0 0>, Index=$System.GameModeID, Num=::NewEnumerator2)) {module=AddVelocity001}
   &PartitionParticles(Number Of Partitions=$Emitter.PortCount)
   &DLR_ProductPreviews(GameModeID:NiagaraInt32=$System.GameModeID, bPreviewVehicleSpeed:NiagaraBool=$System.bPreviewVehicleSpeed)
  ]
  particleUpdateScript [
   &ParticleState()
   &InheritSourceMovement(
    Applied Position Scale=&Multiply_VectorByFloat(Vector=<1 1 1>, Float=1)
    ApplyForce=false
    ApplyPosition=true
    ApplyPositionFalloff=false
   ) {enabled=false, module=InheritSourceMovement001}
   &InheritSourceMovement(
    Applied Position Scale=&Multiply_VectorByFloat(
     Vector=<1 1 1>
     Float=&FloatFromCurve(FloatCurve=DataInterface() {propertyValues {Curve {defaultValue=1, keys [{tangentMode=RCTM_User, time=0, value=1}, {tangentMode=RCTM_User, time=1, value=0}]}}})
    )
    ApplyForce=false
    ApplyPosition=true
    ApplyPositionFalloff=false
   )
   &SolveForcesAndVelocity(Clamp Velocity=false, Limit Acceleration=false)
   /SetVariables(Particles.WPO:NiagaraPosition=&AddVectorToPosition(
     Position=$Particles.Position
     Vector=&TransformVector(
      Vector=&MakeVector(
       X=0
       Y=0
       Z=&Sine(
        Normalized Angle=&Add_Float(
         A=&Multiply_Float(A=$Particles.NormalizedAge, B=-1)
         B=&Multiply_Float(A=$Emitter.LoopedAge, B=0.2)
        )
        Period=&MaskFloatBySpawnGroup(Masked Float=0.5, Passthrough Float=-0.5)
        Scale=&FloatFromCurve(
         FloatCurve=DataInterface() {propertyValues {Curve {
            defaultValue=1
            keys [{interpMode=RCIM_Cubic, tangentMode=RCTM_User, time=0, value=0}, {interpMode=RCIM_Cubic, tangentMode=RCTM_User, time=1, value=1}]
           }}}
         Scale Curve=10
        )
        Bias=0
       )
      )
      Destination Space=::Simulation
      Apply Scale=true
     )
    )) {isSetParameters=true, module=SetVariables_#4}
   &ScaleRibbonWidth(Ribbon Width Scale=&FloatFromCurve(FloatCurve=DataInterface() {propertyValues {Curve {
        defaultValue=1
        keys [
         {arriveTangent=-3.2909, interpMode=RCIM_Cubic, leaveTangent=-3.2909, tangentMode=RCTM_User, time=0, value=1}
         {arriveTangent=-3.2909, interpMode=RCIM_Cubic, leaveTangent=-3.2909, tangentMode=RCTM_User, time=0.1363, value=0.6264}
         {arriveTangent=-3.2909, interpMode=RCIM_Cubic, leaveTangent=-3.2909, tangentMode=RCTM_User, time=0.5, value=1}
         {arriveTangent=-5.5749, interpMode=RCIM_Cubic, leaveTangent=-5.5749, tangentMode=RCTM_User, time=0.7584, value=0.8938}
         {arriveTangent=-3.2909, tangentMode=RCTM_User, time=1, value=0}
        ]
       }}})) {module=ScaleRibbonWidth001}
   &Color(Color=&ColorFromCurve(ColorCurve=DataInterface() {propertyValues {
       AlphaCurve {keys [<at 0 0.005 RCIM_Cubic>,<at 0.5224 1>,<at 1 0 RCIM_Cubic>]}
       BlueCurve {keys [<at 0 0.2363 RCIM_Cubic>,<at 0.2226 0.3615>,<at 0.3389 0.2363>,<at 0.4679 0.3566>,<at 0.6096 0.2384>,<at 0.7241 0.3515>,<at 1 0.3043 RCIM_Cubic>]}
       GreenCurve {keys [<at 0 0.0017 RCIM_Cubic>,<at 0.2226 0.2448>,<at 0.3389 0.0017>,<at 0.4679 0.2417>,<at 0.6096 0.0044>,<at 0.7241 0.2384>,<at 1 0.3335 RCIM_Cubic>]}
       RedCurve {keys [<at 0 3 RCIM_Cubic>,<at 0.2226 0.599>,<at 0.3389 3>,<at 0.4679 0.5917>,<at 0.6096 1>,<at 0.7241 0.5841>,<at 1 1 RCIM_Cubic>]}
      }}))
   &VC_UpdateRibbon_v1(Enable Skeletal Position=true)
   /SetVariables(
    Particles.R1Width:NiagaraFloat=&Multiply_Float(
     A=&SelectFloatFromArray(Direct Array Index=&SpawnGroup(), Float Selection Array=DataInterface() {propertyValues {FloatData [0,-50]}})
     B=&FloatFromCurve(
      FloatCurve=DataInterface() {propertyValues {Curve {
         defaultValue=1
         keys [
          {arriveTangent=-3.4663, interpMode=RCIM_Cubic, leaveTangent=-3.4663, tangentMode=RCTM_User, time=0, value=1}
          {arriveTangent=-2.9168, interpMode=RCIM_Cubic, leaveTangent=-2.9168, tangentMode=RCTM_User, time=0.1363, value=0.6264}
          {arriveTangent=-3.5856, interpMode=RCIM_Cubic, leaveTangent=-3.5856, tangentMode=RCTM_User, time=0.5, value=1}
          {arriveTangent=-6.0095, interpMode=RCIM_Cubic, leaveTangent=-6.0095, tangentMode=RCTM_User, time=0.7381, value=0.8938}
          {tangentMode=RCTM_User, time=1, value=0}
         ]
        }}}
      Scale Curve=1.5
     )
    )
    Particles.R2Width:NiagaraFloat=&Multiply_Float(
     A=&SelectFloatFromArray(Direct Array Index=&SpawnGroup(), Float Selection Array=DataInterface() {propertyValues {FloatData [50,0]}})
     B=&FloatFromCurve(
      FloatCurve=DataInterface() {propertyValues {Curve {
         defaultValue=1
         keys [
          {arriveTangent=-3.4791, interpMode=RCIM_Cubic, leaveTangent=-3.4791, tangentMode=RCTM_User, time=0, value=1}
          {arriveTangent=-3.4392, interpMode=RCIM_Cubic, leaveTangent=-3.4392, tangentMode=RCTM_User, time=0.1363, value=0.6264}
          {arriveTangent=-3.4351, interpMode=RCIM_Cubic, leaveTangent=-3.4351, tangentMode=RCTM_User, time=0.5, value=1}
          {arriveTangent=-6.5944, interpMode=RCIM_Cubic, leaveTangent=-6.5944, tangentMode=RCTM_User, time=0.7381, value=0.8938}
          {tangentMode=RCTM_User, time=1, value=0}
         ]
        }}}
      Scale Curve=1.5
     )
    )
   ) {isSetParameters=true}
  ]
  rendererClasses [&NiagaraRibbonRendererProperties]
  renderers [
   {rendererClass=&NiagaraRibbonRendererProperties}
   {rendererClass=&NiagaraRibbonRendererProperties, rendererIndex=1}
   {rendererClass=&NiagaraRibbonRendererProperties, rendererIndex=2}
   {rendererClass=&NiagaraRibbonRendererProperties, rendererIndex=3}
   {rendererClass=&NiagaraRibbonRendererProperties, rendererIndex=4}
  ]
 }
 RibbonTrail002 {
  emitterUpdateScript [
   &EmitterState()
   &VC_FilteredSocketCount()
   &VC_RateQualityHandler(Default Rate:NiagaraFloat=&MergeHelper_FloatSwitch(0=40, 1=40, 2=40, 3=60, Index=$System.GameModeID, Num=::NewEnumerator2))
   &VC_SpawnRibbon_v1(Number of Ribbons=$Emitter.PortCount, Spawn Time=-1, SpawnRate=$Emitter.Spawn Rate)
  ]
  particleSpawnScript [
   &VC_PaintedVariantsOverride() {enabled=false}
   &VC_InitializeRibbon_v1(
    Color=<rgba 2 0.71 1.3495 1>
    Custom Offset=true
    Enable Skeletal Position=true
    Lifetime=&MergeHelper_FloatSwitch(0=0.74, 1=0.74, 2=0.74, 3=0.4, Index=$System.GameModeID, Num=::NewEnumerator2)
    Local Vector Offset=&AddVectorToPosition(Position=<-30 0 0>, Vector=$User.DTCarOffset)
    Offset Amount=1
    Ribbon Width=80
    Write Material Random=false
   )
   &AddVelocity(Velocity=&MergeHelper_VectorSwitch(0=<-1300 0 0>, 1=<-1300 0 0>, 2=<-1300 0 0>, 3=<-2300 0 0>, Index=$System.GameModeID, Num=::NewEnumerator2)) {module=AddVelocity001}
   &DLR_ProductPreviews(GameModeID:NiagaraInt32=$System.GameModeID, bPreviewVehicleSpeed:NiagaraBool=$System.bPreviewVehicleSpeed)
  ]
  particleUpdateScript [
   &ParticleState()
   &InheritSourceMovement(
    Applied Position Scale=&MergeHelper_VectorSwitch(0=<1 1 1>, 1=<1 1 1>, 2=<1 1 0>, 3=<0.5 0.5 0.5>, Index=$System.GameModeID, Num=::NewEnumerator2)
    ApplyForce=false
    ApplyPosition=true
    ApplyPositionFalloff=false
   ) {enabled=false, module=InheritSourceMovement001}
   &InheritSourceMovement(
    Applied Position Scale=&Multiply_VectorByFloat(
     Vector=<1 1 1>
     Float=&FloatFromCurve(FloatCurve=DataInterface() {propertyValues {Curve {defaultValue=1, keys [{tangentMode=RCTM_User, time=0, value=1}, {tangentMode=RCTM_User, time=1, value=0}]}}})
    )
    ApplyForce=false
    ApplyPosition=true
    ApplyPositionFalloff=false
   )
   &SolveForcesAndVelocity(Clamp Velocity=false, Limit Acceleration=false)
   /SetVariables(Particles.WPO:NiagaraPosition=&AddVectorToPosition(
     Position=$Particles.Position
     Vector=&TransformVector(
      Vector=&MakeVector(
       X=0
       Y=0
       Z=&Sine(
        Normalized Angle=&Add_Float(
         A=&Multiply_Float(A=$Particles.NormalizedAge, B=-1)
         B=&Multiply_Float(A=$Emitter.LoopedAge, B=0.2)
        )
        Period=&MaskFloatBySpawnGroup(Masked Float=0.5, Passthrough Float=-0.5)
        Scale=&FloatFromCurve(
         FloatCurve=DataInterface() {propertyValues {Curve {
            defaultValue=0
            keys [{interpMode=RCIM_Cubic, tangentMode=RCTM_User, time=0, value=0}, {interpMode=RCIM_Cubic, tangentMode=RCTM_User, time=1, value=1}]
           }}}
         Scale Curve=10
        )
        Bias=0
       )
      )
      Destination Space=::Simulation
      Apply Scale=true
     )
    )) {isSetParameters=true, module=SetVariables_#4}
   &ScaleRibbonWidth(Ribbon Width Scale=&FloatFromCurve(FloatCurve=DataInterface() {propertyValues {Curve {
        defaultValue=1
        keys [
         {arriveTangent=-3.2909, interpMode=RCIM_Cubic, leaveTangent=-3.2909, tangentMode=RCTM_User, time=0, value=1}
         {arriveTangent=-3.1648, interpMode=RCIM_Cubic, leaveTangent=-3.1648, tangentMode=RCTM_User, time=0.1363, value=0.6264}
         {arriveTangent=-3.2909, interpMode=RCIM_Cubic, leaveTangent=-3.2909, tangentMode=RCTM_User, time=0.5, value=1}
         {arriveTangent=-5.5749, interpMode=RCIM_Cubic, leaveTangent=-5.5749, tangentMode=RCTM_User, time=0.7584, value=0.8938}
         {arriveTangent=-3.2909, tangentMode=RCTM_User, time=1, value=0}
        ]
       }}})) {module=ScaleRibbonWidth001}
   &VC_UpdateRibbon_v1(Enable Skeletal Position=true)
   /SetVariables(
    Particles.R1Width:NiagaraFloat=&Multiply_Float(
     A=&SelectFloatFromArray(Direct Array Index=&SpawnGroup(), Float Selection Array=DataInterface() {propertyValues {FloatData [0,-50]}})
     B=&FloatFromCurve(
      FloatCurve=DataInterface() {propertyValues {Curve {
         defaultValue=1
         keys [
          {arriveTangent=-3.4663, interpMode=RCIM_Cubic, leaveTangent=-3.4663, tangentMode=RCTM_User, time=0, value=1}
          {arriveTangent=-2.9168, interpMode=RCIM_Cubic, leaveTangent=-2.9168, tangentMode=RCTM_User, time=0.1363, value=0.6264}
          {arriveTangent=-3.5856, interpMode=RCIM_Cubic, leaveTangent=-3.5856, tangentMode=RCTM_User, time=0.5, value=1}
          {arriveTangent=-6.0095, interpMode=RCIM_Cubic, leaveTangent=-6.0095, tangentMode=RCTM_User, time=0.7381, value=0.8938}
          {tangentMode=RCTM_User, time=1, value=0}
         ]
        }}}
      Scale Curve=1.5
     )
    )
    Particles.R2Width:NiagaraFloat=&Multiply_Float(
     A=&SelectFloatFromArray(Direct Array Index=&SpawnGroup(), Float Selection Array=DataInterface() {propertyValues {FloatData [50,0]}})
     B=&FloatFromCurve(
      FloatCurve=DataInterface() {propertyValues {Curve {
         defaultValue=1
         keys [
          {arriveTangent=-3.4791, interpMode=RCIM_Cubic, leaveTangent=-3.4791, tangentMode=RCTM_User, time=0, value=1}
          {arriveTangent=-3.4392, interpMode=RCIM_Cubic, leaveTangent=-3.4392, tangentMode=RCTM_User, time=0.1363, value=0.6264}
          {arriveTangent=-3.4351, interpMode=RCIM_Cubic, leaveTangent=-3.4351, tangentMode=RCTM_User, time=0.5, value=1}
          {arriveTangent=-6.5944, interpMode=RCIM_Cubic, leaveTangent=-6.5944, tangentMode=RCTM_User, time=0.7381, value=0.8938}
          {tangentMode=RCTM_User, time=1, value=0}
         ]
        }}}
      Scale Curve=1.5
     )
    )
   ) {isSetParameters=true}
  ]
  rendererClasses [&NiagaraRibbonRendererProperties]
  renderers [
   {rendererClass=&NiagaraRibbonRendererProperties}
   {rendererClass=&NiagaraRibbonRendererProperties, rendererIndex=1}
  ]
 }
 RibbonTrail005 {
  emitterUpdateScript [
   &EmitterState()
   &VC_FilteredSocketCount()
   &VC_RateQualityHandler(Default Rate:NiagaraFloat=&MergeHelper_FloatSwitch(0=40, 1=40, 2=40, 3=60, Index=$System.GameModeID, Num=::NewEnumerator2))
   &VC_SpawnRibbon_v1(Number of Ribbons=$Emitter.PortCount, Spawn Time=-1, SpawnRate=$Emitter.Spawn Rate)
  ]
  particleSpawnScript [
   &VC_PaintedVariantsOverride() {enabled=false}
   &VC_InitializeRibbon_v1(
    Color=<rgba 2.6406 1.5617 0.165 1>
    Custom Offset=true
    Enable Skeletal Position=true
    Lifetime=&MergeHelper_FloatSwitch(0=0.74, 1=0.74, 2=0.74, 3=0.4, Index=$System.GameModeID, Num=::NewEnumerator2)
    Local Vector Offset=&AddVectorToPosition(Position=<-30 0 0>, Vector=$User.DTCarOffset)
    Offset Amount=1
    Ribbon Width=80
    Write Material Random=false
   )
   &AddVelocity(Velocity=&MergeHelper_VectorSwitch(0=<-1300 0 0>, 1=<-1300 0 0>, 2=<-1300 0 0>, 3=<-2300 0 0>, Index=$System.GameModeID, Num=::NewEnumerator2)) {module=AddVelocity001}
   &PartitionParticles(Number Of Partitions=$Emitter.PortCount)
   &DLR_ProductPreviews(GameModeID:NiagaraInt32=$System.GameModeID, bPreviewVehicleSpeed:NiagaraBool=$System.bPreviewVehicleSpeed)
  ]
  particleUpdateScript [
   &ParticleState()
   &InheritSourceMovement(
    Applied Position Scale=&Multiply_VectorByFloat(Vector=<1 1 1>, Float=1)
    ApplyForce=false
    ApplyPosition=true
    ApplyPositionFalloff=false
   ) {enabled=false, module=InheritSourceMovement001}
   &InheritSourceMovement(
    Applied Position Scale=&Multiply_VectorByFloat(
     Vector=<1 1 1>
     Float=&FloatFromCurve(FloatCurve=DataInterface() {propertyValues {Curve {defaultValue=1, keys [{tangentMode=RCTM_User, time=0, value=1}, {tangentMode=RCTM_User, time=1, value=0}]}}})
    )
    ApplyForce=false
    ApplyPosition=true
    ApplyPositionFalloff=false
   )
   &SolveForcesAndVelocity(Clamp Velocity=false, Limit Acceleration=false)
   /SetVariables(Particles.WPO:NiagaraPosition=&AddVectorToPosition(
     Position=$Particles.Position
     Vector=&TransformVector(
      Vector=&MakeVector(
       X=0
       Y=0
       Z=&Sine(
        Normalized Angle=&Add_Float(
         A=&Multiply_Float(A=$Particles.NormalizedAge, B=-1)
         B=&Multiply_Float(A=$Emitter.LoopedAge, B=0.2)
        )
        Period=&MaskFloatBySpawnGroup(Masked Float=0.5, Passthrough Float=-0.5)
        Scale=&FloatFromCurve(
         FloatCurve=DataInterface() {propertyValues {Curve {
            defaultValue=1
            keys [{interpMode=RCIM_Cubic, tangentMode=RCTM_User, time=0, value=0}, {interpMode=RCIM_Cubic, tangentMode=RCTM_User, time=1, value=1}]
           }}}
         Scale Curve=10
        )
        Bias=0
       )
      )
      Destination Space=::Simulation
      Apply Scale=true
     )
    )) {isSetParameters=true, module=SetVariables_#4}
   &ScaleRibbonWidth(Ribbon Width Scale=&FloatFromCurve(FloatCurve=DataInterface() {propertyValues {Curve {
        defaultValue=1
        keys [
         {arriveTangent=-3.2909, interpMode=RCIM_Cubic, leaveTangent=-3.2909, tangentMode=RCTM_User, time=0, value=1}
         {arriveTangent=-3.2909, interpMode=RCIM_Cubic, leaveTangent=-3.2909, tangentMode=RCTM_User, time=0.1363, value=0.6264}
         {arriveTangent=-3.2909, interpMode=RCIM_Cubic, leaveTangent=-3.2909, tangentMode=RCTM_User, time=0.5, value=1}
         {arriveTangent=-5.5749, interpMode=RCIM_Cubic, leaveTangent=-5.5749, tangentMode=RCTM_User, time=0.7584, value=0.8938}
         {arriveTangent=-3.2909, tangentMode=RCTM_User, time=1, value=0}
        ]
       }}})) {module=ScaleRibbonWidth001}
   &DynamicMaterialParameters(
    Index 2 Param 3=&MaskFloat(Mask Source=1, Mask Parameter=$Particles.PartitionParticles.Partition, Mask=0, Masked Float=1.9473, Passthrough Float=1)
    Index 2 Param 4=0.944
    Module.Index0Float0Write=true
    Module.Index0Float1Write=true
    Module.Index0Float2Write=true
    Module.Index0Float3Write=true
    Module.Index2Float0Write=true
    Module.Index2Float1Write=true
    Module.Index2Float2Write=true
    Module.Index2Float3Write=true
    Write Parameter Index 2=::NewEnumerator1
   ) {enabled=false}
   &Color(Color=&ColorFromCurve(ColorCurve=DataInterface() {propertyValues {
       AlphaCurve {keys [<at 0 0 RCIM_Cubic>,<at 0.4897 1 RCIM_Cubic>,<at 1 0>]}
       BlueCurve {keys [
         <at 0 0 RCIM_Cubic>
         <at 0.4043 0.2422>
         <at 0.5442 0.0216>
         <at 0.715 0.3185>
         {arriveTangent=-0.1183, interpMode=RCIM_Cubic, leaveTangent=-0.1183, time=0.8204, value=0.0331}
         <at 1 0.2848>
        ]}
       GreenCurve {keys [
         <at 0 0.5455 RCIM_Cubic>
         <at 0.4043 0.8394>
         <at 0.5442 0.4292>
         <at 0.715 1>
         {arriveTangent=-0.1694, interpMode=RCIM_Cubic, leaveTangent=-0.1694, time=0.8204, value=0.4564}
         <at 1 0.9517>
        ]}
       RedCurve {keys [
         <at 0 1 RCIM_Cubic>
         <at 0.4043 1.5>
         <at 0.5442 1>
         <at 0.715 1>
         {arriveTangent=0.7017, interpMode=RCIM_Cubic, leaveTangent=0.7017, time=0.8204, value=1}
         <at 1 1.2>
        ]}
      }}))
   &VC_UpdateRibbon_v1(Enable Skeletal Position=true)
   /SetVariables(
    Particles.R1Width:NiagaraFloat=&Multiply_Float(
     A=&SelectFloatFromArray(Direct Array Index=&SpawnGroup(), Float Selection Array=DataInterface() {propertyValues {FloatData [0,-50]}})
     B=&FloatFromCurve(
      FloatCurve=DataInterface() {propertyValues {Curve {
         defaultValue=1
         keys [
          {arriveTangent=-3.4663, interpMode=RCIM_Cubic, leaveTangent=-3.4663, tangentMode=RCTM_User, time=0, value=1}
          {arriveTangent=-2.9168, interpMode=RCIM_Cubic, leaveTangent=-2.9168, tangentMode=RCTM_User, time=0.1363, value=0.6264}
          {arriveTangent=-3.5856, interpMode=RCIM_Cubic, leaveTangent=-3.5856, tangentMode=RCTM_User, time=0.5, value=1}
          {arriveTangent=-6.0095, interpMode=RCIM_Cubic, leaveTangent=-6.0095, tangentMode=RCTM_User, time=0.7381, value=0.8938}
          {tangentMode=RCTM_User, time=1, value=0}
         ]
        }}}
      Scale Curve=1.5
     )
    )
    Particles.R2Width:NiagaraFloat=&Multiply_Float(
     A=&SelectFloatFromArray(Direct Array Index=&SpawnGroup(), Float Selection Array=DataInterface() {propertyValues {FloatData [50,0]}})
     B=&FloatFromCurve(
      FloatCurve=DataInterface() {propertyValues {Curve {
         defaultValue=1
         keys [
          {arriveTangent=-3.4791, interpMode=RCIM_Cubic, leaveTangent=-3.4791, tangentMode=RCTM_User, time=0, value=1}
          {arriveTangent=-3.4392, interpMode=RCIM_Cubic, leaveTangent=-3.4392, tangentMode=RCTM_User, time=0.1363, value=0.6264}
          {arriveTangent=-3.4351, interpMode=RCIM_Cubic, leaveTangent=-3.4351, tangentMode=RCTM_User, time=0.5, value=1}
          {arriveTangent=-6.5944, interpMode=RCIM_Cubic, leaveTangent=-6.5944, tangentMode=RCTM_User, time=0.7381, value=0.8938}
          {tangentMode=RCTM_User, time=1, value=0}
         ]
        }}}
      Scale Curve=1.5
     )
    )
   ) {isSetParameters=true}
  ]
  rendererClasses [&NiagaraRibbonRendererProperties]
  renderers [
   {rendererClass=&NiagaraRibbonRendererProperties}
   {rendererClass=&NiagaraRibbonRendererProperties, rendererIndex=1}
   {rendererClass=&NiagaraRibbonRendererProperties, rendererIndex=2}
   {rendererClass=&NiagaraRibbonRendererProperties, rendererIndex=3}
   {rendererClass=&NiagaraRibbonRendererProperties, rendererIndex=4}
  ]
 }
 RibbonTrail006 {
  emitterUpdateScript [
   &EmitterState()
   &VC_FilteredSocketCount()
   &VC_RateQualityHandler(Default Rate:NiagaraFloat=&MergeHelper_FloatSwitch(0=40, 1=40, 2=40, 3=60, Index=$System.GameModeID, Num=::NewEnumerator2))
   &VC_SpawnRibbon_v1(Number of Ribbons=$Emitter.PortCount, Spawn Time=-1, SpawnRate=$Emitter.Spawn Rate)
  ]
  particleSpawnScript [
   &VC_PaintedVariantsOverride() {enabled=false}
   &VC_InitializeRibbon_v1(
    Color=<rgba 1 0 0.0101 2>
    Custom Offset=true
    Enable Skeletal Position=true
    Lifetime=&MergeHelper_FloatSwitch(0=0.74, 1=0.74, 2=0.74, 3=0.4, Index=$System.GameModeID, Num=::NewEnumerator2)
    Local Vector Offset=&AddVectorToPosition(Position=<-30 0 0>, Vector=$User.DTCarOffset)
    Offset Amount=1
    Ribbon Width=80
    Write Material Random=false
   )
   &AddVelocity(Velocity=&MergeHelper_VectorSwitch(0=<-1300 0 0>, 1=<-1300 0 0>, 2=<-1300 0 0>, 3=<-2300 0 0>, Index=$System.GameModeID, Num=::NewEnumerator2)) {module=AddVelocity001}
   &PartitionParticles(Number Of Partitions=$Emitter.PortCount)
   &DLR_ProductPreviews(GameModeID:NiagaraInt32=$System.GameModeID, bPreviewVehicleSpeed:NiagaraBool=$System.bPreviewVehicleSpeed)
  ]
  particleUpdateScript [
   &ParticleState()
   &InheritSourceMovement(
    Applied Position Scale=&Multiply_VectorByFloat(Vector=<1 1 1>, Float=1)
    ApplyForce=false
    ApplyPosition=true
    ApplyPositionFalloff=false
   ) {enabled=false, module=InheritSourceMovement001}
   &InheritSourceMovement(
    Applied Position Scale=&Multiply_VectorByFloat(
     Vector=<1 1 1>
     Float=&FloatFromCurve(FloatCurve=DataInterface() {propertyValues {Curve {defaultValue=1, keys [{tangentMode=RCTM_User, time=0, value=1}, {tangentMode=RCTM_User, time=1, value=0}]}}})
    )
    ApplyForce=false
    ApplyPosition=true
    ApplyPositionFalloff=false
   )
   &SolveForcesAndVelocity(Clamp Velocity=false, Limit Acceleration=false)
   /SetVariables(Particles.WPO:NiagaraPosition=&AddVectorToPosition(
     Position=$Particles.Position
     Vector=&TransformVector(
      Vector=&MakeVector(
       X=0
       Y=0
       Z=&Sine(
        Normalized Angle=&Add_Float(
         A=&Multiply_Float(A=$Particles.NormalizedAge, B=-1)
         B=&Multiply_Float(A=$Emitter.LoopedAge, B=0.2)
        )
        Period=&MaskFloatBySpawnGroup(Masked Float=0.5, Passthrough Float=-0.5)
        Scale=&FloatFromCurve(
         FloatCurve=DataInterface() {propertyValues {Curve {
            defaultValue=1
            keys [{interpMode=RCIM_Cubic, tangentMode=RCTM_User, time=0, value=0}, {interpMode=RCIM_Cubic, tangentMode=RCTM_User, time=1, value=1}]
           }}}
         Scale Curve=10
        )
        Bias=0
       )
      )
      Destination Space=::Simulation
      Apply Scale=true
     )
    )) {isSetParameters=true, module=SetVariables_#4}
   &ScaleRibbonWidth(Ribbon Width Scale=&FloatFromCurve(FloatCurve=DataInterface() {propertyValues {Curve {
        defaultValue=1
        keys [
         {arriveTangent=-3.2909, interpMode=RCIM_Cubic, leaveTangent=-3.2909, tangentMode=RCTM_User, time=0, value=1}
         {arriveTangent=-3.2909, interpMode=RCIM_Cubic, leaveTangent=-3.2909, tangentMode=RCTM_User, time=0.1363, value=0.6264}
         {arriveTangent=-3.2909, interpMode=RCIM_Cubic, leaveTangent=-3.2909, tangentMode=RCTM_User, time=0.5, value=1}
         {arriveTangent=-5.5749, interpMode=RCIM_Cubic, leaveTangent=-5.5749, tangentMode=RCTM_User, time=0.7584, value=0.8938}
         {arriveTangent=-3.2909, tangentMode=RCTM_User, time=1, value=0}
        ]
       }}})) {module=ScaleRibbonWidth001}
   &Color(Color=&ColorFromCurve(ColorCurve=DataInterface() {propertyValues {
       AlphaCurve {keys [<at 0 0>,<at 0.5115 1 RCIM_Cubic>,<at 1 0 RCIM_Cubic>]}
       BlueCurve {keys [<at 0 1.0089 RCIM_Cubic>,<at 0.1753 0.7075>,<at 0.2971 0.2374>,<at 0.5006 0.0015>,<at 0.7277 0.2486>,<at 0.864 0.8523>,<at 1 0.1176 RCIM_Cubic>]}
       GreenCurve {keys [<at 0 0 RCIM_Cubic>,<at 0.1753 0.0355>,<at 0.2971 0.4154>,<at 0.5006 0.1188>,<at 0.7277 0.4295>,<at 0.864 0.0185>,<at 1 0 RCIM_Cubic>]}
       RedCurve {keys [<at 0 3 RCIM_Cubic>,<at 0.1753 3>,<at 0.2971 3>,<at 0.5006 3>,<at 0.7277 3>,<at 0.864 3>,<at 1 1 RCIM_Cubic>]}
      }}))
   &VC_UpdateRibbon_v1(Enable Skeletal Position=true)
   /SetVariables(
    Particles.R1Width:NiagaraFloat=&Multiply_Float(
     A=&SelectFloatFromArray(Direct Array Index=&SpawnGroup(), Float Selection Array=DataInterface() {propertyValues {FloatData [0,-50]}})
     B=&FloatFromCurve(
      FloatCurve=DataInterface() {propertyValues {Curve {
         defaultValue=1
         keys [
          {arriveTangent=-3.4663, interpMode=RCIM_Cubic, leaveTangent=-3.4663, tangentMode=RCTM_User, time=0, value=1}
          {arriveTangent=-2.9168, interpMode=RCIM_Cubic, leaveTangent=-2.9168, tangentMode=RCTM_User, time=0.1363, value=0.6264}
          {arriveTangent=-3.5856, interpMode=RCIM_Cubic, leaveTangent=-3.5856, tangentMode=RCTM_User, time=0.5, value=1}
          {arriveTangent=-6.0095, interpMode=RCIM_Cubic, leaveTangent=-6.0095, tangentMode=RCTM_User, time=0.7381, value=0.8938}
          {tangentMode=RCTM_User, time=1, value=0}
         ]
        }}}
      Scale Curve=1.5
     )
    )
    Particles.R2Width:NiagaraFloat=&Multiply_Float(
     A=&SelectFloatFromArray(Direct Array Index=&SpawnGroup(), Float Selection Array=DataInterface() {propertyValues {FloatData [50,0]}})
     B=&FloatFromCurve(
      FloatCurve=DataInterface() {propertyValues {Curve {
         defaultValue=1
         keys [
          {arriveTangent=-3.4791, interpMode=RCIM_Cubic, leaveTangent=-3.4791, tangentMode=RCTM_User, time=0, value=1}
          {arriveTangent=-3.4392, interpMode=RCIM_Cubic, leaveTangent=-3.4392, tangentMode=RCTM_User, time=0.1363, value=0.6264}
          {arriveTangent=-3.4351, interpMode=RCIM_Cubic, leaveTangent=-3.4351, tangentMode=RCTM_User, time=0.5, value=1}
          {arriveTangent=-6.5944, interpMode=RCIM_Cubic, leaveTangent=-6.5944, tangentMode=RCTM_User, time=0.7381, value=0.8938}
          {tangentMode=RCTM_User, time=1, value=0}
         ]
        }}}
      Scale Curve=1.5
     )
    )
   ) {isSetParameters=true}
  ]
  rendererClasses [&NiagaraRibbonRendererProperties]
  renderers [
   {rendererClass=&NiagaraRibbonRendererProperties}
   {rendererClass=&NiagaraRibbonRendererProperties, rendererIndex=1}
   {rendererClass=&NiagaraRibbonRendererProperties, rendererIndex=2}
   {rendererClass=&NiagaraRibbonRendererProperties, rendererIndex=3}
   {rendererClass=&NiagaraRibbonRendererProperties, rendererIndex=4}
  ]
 }
}
summary {
 systemName=NS_Boost_Trippy
 emitters [
  {emitterName=BlueStars002, rendererClasses [&NiagaraSpriteRendererProperties]}
  {emitterName=RibbonTrail005, rendererClasses [&NiagaraRibbonRendererProperties]}
  {emitterName=RibbonTrail006, rendererClasses [&NiagaraRibbonRendererProperties]}
  {emitterName=RibbonTrail001, rendererClasses [&NiagaraRibbonRendererProperties]}
  {emitterName=RibbonTrail002, rendererClasses [&NiagaraRibbonRendererProperties]}
 ]
 userVariables [
  User.bIsFrontendPreview {type=NiagaraBool, defaultValue=true}
  User.bIsPainted {type=NiagaraBool}
  User.blsIsCoroColor {type=NiagaraBool}
  User.bPreviewVehicleSpeed {type=NiagaraBool}
  User.ColorVariantID {type=NiagaraInt32}
  User.ColorWhite {type=NiagaraFloat, defaultValue=0, description=""}
  User.DTCarOffset {type=Vector3f, description="", defaultValue {struct=Vector3f, value=<0 0 0>}}
  User.DTCarOffset003 {type=NiagaraPosition, description="", defaultValue {}}
  User.Emissive {type=NiagaraFloat, defaultValue=10, description=""}
  User.EmissiveTrail {type=NiagaraFloat, defaultValue=1, description=""}
  User.EmissiveTrailAlpha {type=NiagaraFloat, defaultValue=0.6, description=""}
  User.GameModeID {type=NiagaraInt32}
  User.PaintedColor {type=LinearColor, description="", defaultValue {struct=LinearColor, value=<rgba 1 1 1 1>}}
  User.PaintedScalar {type=NiagaraFloat, defaultValue=1, description=""}
  User.PortCount {type=NiagaraInt32, defaultValue=2}
  User.Shape01 {type=LinearColor, description="", defaultValue {struct=LinearColor, value=<rgba 0 0.1275 1 1>}}
  User.Shape02 {type=LinearColor, description="", defaultValue {struct=LinearColor, value=<rgba 1 0 0.0097 1>}}
  User.WaveMaxScale {type=NiagaraFloat, defaultValue=1, description=""}
  User.WavePeriod {type=NiagaraFloat, defaultValue=1, description=""}
 ]
}
